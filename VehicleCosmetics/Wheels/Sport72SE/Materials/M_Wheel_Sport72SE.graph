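# Masked wheel material for the Sport72SE vehicle cosmetic, built on MF_WheelParent_Simple with animated emissive effects added.
# Opposing scrolling gradient stripes and a radial sine pulse on UV1 provide motion, alongside rim and tire channel masks and an EmissiveColor parameter.
# Inverted red vertex color blends two emissive layers.
# The result overrides emissive, then passes through the shared vehicle damage and de-rez transition functions.
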Material /VehicleCosmetics/Wheels/Sport72SE/Materials/M_Wheel_Sport72SE
graph {
 properties {
  bAutomaticallySetUsageInEditor=false
  bUseMaterialAttributes=true
  bUsedWithMeshDeformer=true
  bUsedWithSkeletalMesh=true
  blendMode=BLEND_Masked
  pixelDepthOffsetMode=PDOM_Legacy
  usedShadingModels=Default Lit
 }
 nodes [
  n1 = MaterialFunctionCall() {materialFunction=&MF_WheelParent_Simple}
  n2 = NamedRerouteUsage()
  n3 = Panner() {speedY=0.75}
  n4 = Abs(None=Add(A=ComponentMask(None=@n3) {g=true}, B=Constant() {r=-0.5}))
  n5 = Power(Base=TextureSample(UVs=Multiply(A=@n4) {constB=0.75}) {texture=/VehicleCosmetics/Wheels/_SharedTextures/FX/GradientStripe_T}.R)
  n6 = TextureCoordinate()
  n7 = Step(Y=OneMinus(None=ComponentMask(None=@n6) {g=true})) {constX=0.5}
  n8 = Panner() {speedY=-0.75}
  n9 = Abs(None=Add(A=ComponentMask(None=@n8) {g=true}, B=Constant() {r=-0.5}))
  n10 = Power(Base=TextureSample(UVs=Multiply(A=@n9) {constB=0.75}) {texture=&GradientStripe_T}.RGB)
  n11 = Step(Y=ComponentMask(None=TextureCoordinate()) {g=true}) {constX=0.5}
  n12 = Multiply(A=NamedRerouteUsage(), B=LinearInterpolate(A=Multiply(A=@n5, B=@n7), B=Multiply(A=@n10, B=@n11)))
  n13 = NamedRerouteUsage()
  n14 = NamedRerouteUsage()
  n15 = Multiply(A=NamedRerouteUsage(), B=Step(X=NamedRerouteUsage()) {constY=0.1})
  n16 = NamedRerouteUsage()
  n17 = Multiply(A=NamedRerouteUsage(), B=Multiply(A=NamedRerouteUsage(), B=@n16))
  n18 = NamedRerouteUsage()
  n19 = Multiply(A=NamedRerouteUsage(), B=@n16)
  n20 = LinearInterpolate(
   A=Add(A=Multiply(A=@n2, B=@n12), B=Multiply(A=@n13, B=@n14))
   B=Add(A=Add(A=@n15, B=@n17) {constB=0}, B=Multiply(A=@n18, B=@n19)) {constB=0}
   Alpha=OneMinus(None=@n1.VertexColorR)
  )
  MaterialFunctionCall(In=MaterialFunctionCall(Material=SetMaterialAttributes(MaterialAttributes=@n1.MaterialAttributes, Emissive Color=@n20) {attributeSetTypes [#1]}) {materialFunction=&MF_BaseVehicle_Damage}.Result) {materialFunction=&MF_VehicleDeRez_Transition}
  n21 = VectorParameter() {group=5_Special Effects, parameterName=EmissiveColor, sortPriority=33, defaultValue=<rgba 0.956 0.6172 0.0015 1>}
  NamedRerouteDeclaration(None=@n21.RGB) {name=EmissiveColor, nodeColor=<rgba 1 0 0.4219 1>}
  n22 = MaterialFunctionCall(Color=@n1.RimRGBAMask) {materialFunction=&SplitComponents}
  NamedRerouteDeclaration(None=@n22.R) {name=RimRMask, nodeColor=<rgba 1 0 0.0937 1>}
  NamedRerouteDeclaration(None=@n22.G) {name=RimGMask, nodeColor=<rgba 0 0.4141 1 1>}
  n23 = TextureCoordinate() {coordinateIndex=1}
  n24 = ComponentMask(None=Multiply(A=Subtract(A=@n23) {constB=0.5}) {constB=3}) {g=true, r=true}
  n25 = Time()
  n26 = ScalarParameter() {defaultValue=-0.4, group=5_Special Effects, parameterName=PulseSpeed, sortPriority=34}
  n27 = Sine(None=Add(A=Length(None=@n24) {is=class MaterialExpressionLength}, B=Multiply(A=@n25, B=@n26)))
  NamedRerouteDeclaration(None=Clamp(None=SmoothStep(Value=@n27) {constMax=2})) {name=PulseFX, nodeColor=<rgba 0 1 0.5078 1>}
  NamedRerouteDeclaration(None=@n22.B) {name=RimBMask, nodeColor=<rgba 0.7344 1 0 1>}
  n28 = MaterialFunctionCall(Color=@n1.TireRGBAMask) {materialFunction=&SplitComponents}
  NamedRerouteDeclaration(None=@n28.R) {name=TireRMask, nodeColor=<rgba 0.9453 0 1 1>}
  NamedRerouteDeclaration(None=@n28.G) {name=TireGMask, nodeColor=<rgba 0 1 0.625 1>}
 ]
}
